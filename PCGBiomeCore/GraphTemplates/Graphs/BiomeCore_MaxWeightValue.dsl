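; PCG graph template from Biome Core that computes a per-point maximum between the MaxValue attribute and a layer weight attribute.
; A single Attribute Maths Op in Max mode takes the Points input on both operands and outputs the resulting points.
; The Layers param feeds the second source selector override, so the attribute compared against MaxValue is chosen by data rather than fixed in the node.
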(graph BiomeCore_MaxWeightValue
  (inputs (Layers :type Param) (Points :type Point))
  (outputs (Points :type Point :advanced))
  (bind AttributeMathsOp_0 (Attribute%20Maths%20Op :InputSource1 MaxValue :InputSource2 None :Operation Max))
  (connect DefaultInputNode Points AttributeMathsOp_0 InA)
  (connect DefaultInputNode Points AttributeMathsOp_0 InB)
  (connect DefaultInputNode Layers AttributeMathsOp_0 InputSource2)
  (output Points AttributeMathsOp_0))
