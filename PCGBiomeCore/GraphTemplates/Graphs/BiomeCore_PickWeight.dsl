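; PCG graph template for Biome Core that performs one step of a weighted pick, assigning a GeneratorSubType to points.
; It subtracts a point attribute, selected through the Layers input, from each point's WeightValue.
; Points with an unset GeneratorSubType whose WeightValue drops below zero take their GeneratorSubType from Layers.
; All points are merged back out, suiting cumulative per-layer passes; the Tolerance parameter is unused here.
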(graph BiomeCore_PickWeight
  (params (Tolerance :type Double :default 0.2))
  (inputs (Layers :type Param) (Points :type Point))
  (outputs (Points :type Point :advanced))
  (bind AttributeMathsOp_0 (Attribute%20Maths%20Op :InputSource1 WeightValue :Operation Subtract))
  (bind AttributeFilter_0 (Filter%20Attribute%20Elements :AttributeTypes (struct :Type Name) :bUseConstantThreshold true :Operator Equal :TargetAttribute GeneratorSubType :ThresholdAttribute @Last))
  (bind AttributeFilter_1 (Filter%20Attribute%20Elements :bUseConstantThreshold true :Operator Lesser :TargetAttribute WeightValue :ThresholdAttribute @Last))
  (bind AttributeMathsOp_2 (Attribute%20Maths%20Op :InputSource1 GeneratorSubType :Operation Set))
  (bind MergePoints_0 (Merge%20Points))
  (connect AttributeMathsOp_0 Out AttributeFilter_0 In)
  (connect AttributeFilter_0 InsideFilter AttributeFilter_1 In)
  (connect DefaultInputNode Points AttributeMathsOp_0 InA)
  (connect DefaultInputNode Points AttributeMathsOp_0 InB)
  (connect DefaultInputNode Layers AttributeMathsOp_0 InputSource2)
  (connect AttributeFilter_1 InsideFilter AttributeMathsOp_2 InA)
  (connect DefaultInputNode Layers AttributeMathsOp_2 InB)
  (connect AttributeFilter_0 OutsideFilter MergePoints_0 In)
  (connect AttributeFilter_1 OutsideFilter MergePoints_0 In)
  (connect AttributeMathsOp_2 Out MergePoints_0 In)
  (output Points MergePoints_0))
