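; PCG Biome Core graph template that scales points according to their density.
; Raises density to DensityExponent, then remaps it into the ScaleMin to ScaleMax range with unit-range clamping enabled.
; Multiplies each input point's existing scale by that remapped value.
; Copies the original density back from the input, so the output keeps its density.
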(graph BiomeCore_ScaleByDensity
  (params (ScaleMin :type Double) (ScaleMax :type Double :default 1.0) (DensityExponent :type Double :default 1.0))
  (inputs (Points :type Point))
  (outputs (Points :type Point))
  (bind AttributeMathsOp_12 (Attribute%20Maths%20Op :InputSource1 $Density :Operation Pow :OutputTarget $Density :InB DensityExponent))
  (bind AttributeRemap_0 (Attribute%20Remap :bClampToUnitRange true :InputSource $Density :OutRangeMax-param ScaleMax :OutRangeMin-param ScaleMin))
  (bind AttributeMathsOp_14 (Attribute%20Maths%20Op :InputSource1 $Scale :Operation Multiply :OutputTarget $Scale)) ; Multiply by $Scale
  (bind TransferAttribute_11 (Copy%20Attributes :InputSource $Density :OutputTarget $Density)) ; Restore points' original $Density value
  (connect DefaultInputNode Points AttributeMathsOp_12 InA)
  (connect DefaultInputNode Points AttributeMathsOp_14 InA)
  (connect AttributeRemap_0 Out AttributeMathsOp_14 InB)
  (connect AttributeMathsOp_12 Out AttributeRemap_0 In)
  (connect DefaultInputNode Points TransferAttribute_11 Source)
  (connect AttributeMathsOp_14 Out TransferAttribute_11 Target)
  (output Points TransferAttribute_11))
