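; PCG graph template for BiomeCore that computes a layer's weight on each point relative to that point's MaxValue.
; Tolerance is clamped and inverted to scale MaxValue into a lower threshold.
; The attribute named by the Layers input is remapped from that threshold-to-MaxValue range, clamped, then accumulated into TotalValue on the output points.
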(graph BiomeCore_SetWeight
  (params (Tolerance :type Double :default 0.1))
  (inputs (Layers :type Param) (Points :type Point))
  (outputs (Points :type Point :advanced))
  (bind AttributeMathsOp_9 (Attribute%20Maths%20Op :Operation Clamp :In Tolerance))
  (bind AttributeMathsOp_5 (Attribute%20Maths%20Op :Operation OneMinus))
  (bind AttributeMathsOp_1 (Attribute%20Maths%20Op :InputSource1 MaxValue :Operation Multiply))
  (bind AttributeMathsOp_2 (Attribute%20Maths%20Op :InputSource1 MaxValue :InputSource2 MaxValue :Operation Subtract))
  (bind AttributeMathsOp_3 (Attribute%20Maths%20Op :InputSource1 Layer_01 :InputSource2 MaxValue :Operation Subtract))
  (bind AttributeMathsOp_4 (Attribute%20Maths%20Op :InputSource1 Layer_01 :InputSource2 MaxValue :Operation Divide))
  (bind AttributeMathsOp_6 (Attribute%20Maths%20Op :Operation Clamp))
  (bind AttributeMathsOp_8 (Attribute%20Maths%20Op :InputSource2 TotalValue :OutputTarget TotalValue))
  (connect DefaultInputNode Points AttributeMathsOp_1 InA)
  (connect AttributeMathsOp_5 Out AttributeMathsOp_1 InB)
  (connect DefaultInputNode Points AttributeMathsOp_2 InA)
  (connect AttributeMathsOp_1 Out AttributeMathsOp_2 InB)
  (connect DefaultInputNode Points AttributeMathsOp_3 InA)
  (connect AttributeMathsOp_1 Out AttributeMathsOp_3 InB)
  (connect DefaultInputNode Layers AttributeMathsOp_3 InputSource1)
  (connect AttributeMathsOp_3 Out AttributeMathsOp_4 InA)
  (connect AttributeMathsOp_2 Out AttributeMathsOp_4 InB)
  (connect DefaultInputNode Layers AttributeMathsOp_4 InputSource1)
  (connect AttributeMathsOp_9 Out AttributeMathsOp_5 In)
  (connect AttributeMathsOp_4 Out AttributeMathsOp_6 In)
  (connect DefaultInputNode Layers AttributeMathsOp_6 InputSource1)
  (connect AttributeMathsOp_6 Out AttributeMathsOp_8 InA)
  (connect AttributeMathsOp_6 Out AttributeMathsOp_8 InB)
  (connect DefaultInputNode Layers AttributeMathsOp_8 InputSource1)
  (output Points AttributeMathsOp_8))
