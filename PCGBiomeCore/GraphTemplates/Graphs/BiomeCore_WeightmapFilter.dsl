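; PCG Biome Core graph that assigns each point a GeneratorSubType from a weightmap layer data asset.
; When the asset's WeightmapFilter is off, points pass through; otherwise a loop over weightmap layers computes MaxValue, which is tested against a small threshold.
; Points inside the filter get tolerance-based weights, then noise scaled by TotalValue drives a weighted random pick across layers.
; Points outside get a default GeneratorSubType, then both sets merge and temporary attributes are removed.
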(graph BiomeCore_WeightmapFilter
  (inputs (WeightmapLayerDataAsset :type Param))
  (bind GetPropertyFromObjectPath_3 (Get%20Property%20From%20Object%20Path :PropertyName WeightmapFilter))
  (bind Branch_0 (Branch))
  (bind AddAttribute_1 (Add%20Attribute :OutputTarget MaxValue))
  (bind GetPropertyFromObjectPath_1 (Get%20Property%20From%20Object%20Path :PropertyName Weightmaps))
  (bind AttributePartition_0 (Attribute%20Partition :PartitionAttributeSelectors [$Index]))
  (bind Loop_1 (Loop :Graph /PCGBiomeCore/GraphTemplates/Graphs/BiomeCore_MaxWeightValue.BiomeCore_MaxWeightValue))
  (bind AddAttribute_2 (Add%20Attribute :OutputTarget TotalValue))
  (bind AttributeFilter_0 (Filter%20Attribute%20Elements :AttributeTypes (struct :DoubleValue 0.01) :bUseConstantThreshold true :TargetAttribute MaxValue :ThresholdAttribute @Last))
  (bind GetPropertyFromObjectPath_2 (Get%20Property%20From%20Object%20Path :PropertyName WeightmapTolerance))
  (bind Loop_2 (Loop :Graph /PCGBiomeCore/GraphTemplates/Graphs/BiomeCore_SetWeight.BiomeCore_SetWeight))
  (bind DeleteAttributes_0 (Delete%20Attributes :SelectedAttributes "MaxValue"))
  (bind AttributeNoise_0 (Attribute%20Noise :OutputTarget WeightValue))
  (bind AttributeMathsOp_0 (Attribute%20Maths%20Op :InputSource1 WeightValue :InputSource2 TotalValue :Operation Multiply))
  (bind AddAttribute_3 (Add%20Attribute :AttributeTypes (struct :Type Name) :OutputTarget GeneratorSubType))
  (bind Loop_3 (Loop :Graph /PCGBiomeCore/GraphTemplates/Graphs/BiomeCore_PickWeight.BiomeCore_PickWeight))
  (bind AddAttribute_5 (Add%20Attribute :AttributeTypes (struct :Type Name) :OutputTarget GeneratorSubType))
  (bind MergePoints_1 (Merge%20Points))
  (bind DeleteAttributes_1 (Delete%20Attributes :SelectedAttributes "WeightValue,TotalValue"))
  (connect Branch_0 Output%20B AddAttribute_1 In)
  (connect Loop_1 Points AddAttribute_2 In)
  (connect AttributeMathsOp_0 Out AddAttribute_3 In)
  (connect AttributeFilter_0 OutsideFilter AddAttribute_5 In)
  (connect AddAttribute_2 Out AttributeFilter_0 In)
  (connect AttributeNoise_0 Out AttributeMathsOp_0 InA)
  (connect AttributeNoise_0 Out AttributeMathsOp_0 InB)
  (connect DeleteAttributes_0 Out AttributeNoise_0 In)
  (connect GetPropertyFromObjectPath_1 Out AttributePartition_0 In)
  (connect GetPropertyFromObjectPath_3 Out Branch_0 bOutputToB)
  (connect DefaultInputNode In Branch_0 In)
  (connect Loop_2 Points DeleteAttributes_0 In)
  (connect MergePoints_1 Out DeleteAttributes_1 In)
  (connect DefaultInputNode WeightmapLayerDataAsset GetPropertyFromObjectPath_1 In)
  (connect DefaultInputNode WeightmapLayerDataAsset GetPropertyFromObjectPath_2 In)
  (connect DefaultInputNode WeightmapLayerDataAsset GetPropertyFromObjectPath_3 In)
  (connect AttributePartition_0 Out Loop_1 Layers)
  (connect AddAttribute_1 Out Loop_1 Points)
  (connect AttributePartition_0 Out Loop_2 Layers)
  (connect AttributeFilter_0 InsideFilter Loop_2 Points)
  (connect GetPropertyFromObjectPath_2 Out Loop_2 Tolerance)
  (connect AttributePartition_0 Out Loop_3 Layers)
  (connect AddAttribute_3 Out Loop_3 Points)
  (connect AddAttribute_5 Out MergePoints_1 In)
  (connect Loop_3 Points MergePoints_1 In)
  (output Out Branch_0.Output%20A)
  (output Out DeleteAttributes_1))
